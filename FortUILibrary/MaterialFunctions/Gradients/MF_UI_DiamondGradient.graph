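# Material function in the UI shader library that generates a diamond gradient from a UV input, which defaults to the texture coordinates.
# It remaps UVs to a centered -1 to 1 range, takes absolute values, and sums both axes through a dot product with one to get Manhattan distance.
# The inverted, saturated distance feeds the Linear output, and its square feeds Smooth for a softer falloff.
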MaterialFunction /FortUILibrary/MaterialFunctions/Gradients/MF_UI_DiamondGradient
graph {
 properties {bExposeToLibrary=true, description=Generate a diamond gradient, previewMaterialDomain=MD_UI, libraryCategoriesText [UI_ShaderLibrary]}
 nodes [
  n1 = TextureCoordinate()
  n2 = Abs(None=ConstantBiasScale(None=FunctionInput(Preview=@n1) {bUsePreviewValueAsDefault=true, inputName=UV, inputType=FunctionInput_Vector2}) {bias=-0.5, scale=2})
  n3 = Constant() {r=1}
  n4 = Saturate(None=OneMinus(None=DotProduct(A=@n2, B=@n3)))
  FunctionOutput(None=@n4) {outputName=Linear}
  FunctionOutput(None=Multiply(A=@n4, B=@n4)) {outputName=Smooth, sortPriority=5}
 ]
}
