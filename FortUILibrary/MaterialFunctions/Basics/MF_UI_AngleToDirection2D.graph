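# UI Library material function that converts a normalized angle (0 to 1 is one full turn) into a unit 2D direction vector.
# The scalar input feeds Cosine and Sine, appended as (cos, sin), so zero points +X and values wrap.
# Its description notes that uniform angles fold to the CPU, making it nearly free per pixel.
# An unconnected usage example builds auto-rotation from Time × speed, Frac and a direction offset.
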MaterialFunction /FortUILibrary/MaterialFunctions/Basics/MF_UI_AngleToDirection2D
graph {
 properties {
  bExposeToLibrary=true
  description="Converts a normalized angle into a unit 2D direction vector. Angle 0–1 = one full 360° turn. Output = (cos, sin), always unit length; angle 0 points +X. Pure and cheap — when the angle is a uniform (parameter or Time-driven), the cos/sin fold to the CPU, so it's almost free per pixel. For continuous rotation, drive the angle with Time × speed (see Usage Example)."
  previewMaterialDomain=MD_UI
  userExposedCaption=MF_UI_AngleToDirection2D
  libraryCategoriesText [UI_Library]
 }
 nodes [
  n1 = Time()
  n2 = Constant() {desc=Auto Rotation Speed}
  Add(A=Constant() {desc="Direction (0 to 1)"}, B=Frac(None=Multiply(A=@n1, B=@n2)))
  n3 = FunctionInput() {
   bUsePreviewValueAsDefault=true
   desc="Normalized angle: 0–1 maps to a full 360° turn \r\n(values outside just wrap, it's periodic). \r\n0 = +X (positive X direction)."
   description="Normalized angle: 0–1 maps to a full 360° turn (values outside just wrap, it's periodic). 0 = +X (positive X direction)."
   inputName="Angle (0 to 1)"
   inputType=FunctionInput_Scalar
   previewValue=<0 0 0 0>
  }
  FunctionOutput(None=AppendVector(A=Cosine(None=@n3), B=Sine(None=@n3))) {
   desc="Unit-length 2D direction (cos, sin). \r\n\r\n0 : (1, 0);  right (+X)\r\n0.25 : (0, 1); +Y\r\n0.5 : (-1, 0);  left (-X)\r\n0.75 : (0, -1). -Y"
   description="Unit-length 2D direction (cos, sin). \r\n\r\n0 : (1, 0);  right (+X)\r\n0.25 : (0, 1); +Y\r\n0.5 : (-1, 0);  left (-X)\r\n0.75 : (0, -1). -Y\r\n"
   outputName="Direction (vector2)"
  }
 ]
}
