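# UI material function that outputs a soft, rotatable angular cone gradient peaking toward a Direction vector, for decorating shapes.
# It centers UVs by an offset and normalizes them to a per-pixel direction, so the falloff follows angle and ignores distance.
# The dot product with Direction is remapped by Spread, which sets cone width, then saturated.
# Direction accepts a fixed vector or MF_UI_AngleToDirection2D output for rotation.
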MaterialFunction /FortUILibrary/MaterialFunctions/Gradients/MF_UI_DirectionalConeGradient
graph {
 properties {
  bExposeToLibrary=true
  description="True angular cone gradient. From centered UVs and a Direction vector, outputs a soft cone that peaks toward Direction and fades with angle. Fully rotatable, no distance breathing. Spread controls how wide the cone opens (0 = nothing to 2 = full ring). Drive Direction with MF_UI_AngleToDirection2D for rotation."
  previewMaterialDomain=MD_UI
  userExposedCaption=MF_UI_DirectionalConeGradient
  libraryCategoriesText [UI_Library]
 }
 nodes [
  Constant2Vector() {g=0.5, r=0.5}
  Constant() {r=2}
  n1 = OneMinus(None=NamedRerouteUsage())
  n2 = Divide(B=Subtract(B=@n1))
  n3 = Normalize(VectorInput=NamedRerouteUsage()) {desc="Normalize keeps only direction and discards magnitude,so scaling the UVs does nothing here (a ConstantBiasScale's Scale would be wasted). \r\nOnly centering matters, which is why we center with a plain Subtract, not a CBS."}
  n4 = Subtract(A=DotProduct(A=NamedRerouteDeclaration(None=@n3) {name=Pixel Direction, nodeColor=<rgba 0 0 0 1>}, B=NamedRerouteUsage()), B=@n1)
  FunctionOutput(None=Saturate(None=Multiply(A=@n4, B=@n2))) {outputName=Directional Cone Gradient, sortPriority=3}
  n5 = Multiply(A=Time(), B=Constant() {desc=Auto Rotation Speed, r=0.25})
  MaterialFunctionCall!("Angle (0 to 1)"=Add(A=Constant() {desc="Direction (0 to 1)"}, B=Frac(None=@n5))) {materialFunction=/FortUILibrary/MaterialFunctions/Basics/MF_UI_AngleToDirection2D}
  n6 = FunctionInput() {
   bUsePreviewValueAsDefault=true
   desc="Control how wide the cone opens (0 = nothing to 2 = full ring)\r\n\r\n0 = cone 0, fill 0, \r\n0.5 = half cone, fill 0, \r\n1 = full faded gradient, fill 0, \r\n1.5 = full faded, fill 0.5, \r\n2 = full faded, fill 1."
   description="Control how wide the cone opens (0 = nothing to 2 = full ring)"
   inputName=Spread
   inputType=FunctionInput_Scalar
   sortPriority=5
   previewValue=<0.25 0 0 0>
  }
  NamedRerouteDeclaration(None=Multiply(A=@n6) {constB=2, desc=remap}) {name=Cone Spread, nodeColor=<rgba 0 0 0 1>}
  n7 = FunctionInput() {
   bUsePreviewValueAsDefault=true
   desc="Direction the cone points toward (normalized internally). \r\n\r\n(-0.7, -0.7) = top-left, \r\n(1, 0) = right. \r\n\r\nStatic V2 for a fixed direction, or drive with MF_UI_AngleToDirection2D for rotation."
   description="Direction the cone points toward (normalized internally). (-0.7, -0.7) = top-left, (1, 0) = right. Static V2 for a fixed direction, or drive with MF_UI_AngleToDirection2D for rotation."
   inputName=Direction
   inputType=FunctionInput_Vector2
   sortPriority=4
   previewValue=<-0.7 -0.7 0 0>
  }
  NamedRerouteDeclaration(None=@n7) {name=Direction, nodeColor=<rgba 0 0 0 1>}
  n8 = FunctionInput(Preview=TextureCoordinate()) {
   bUsePreviewValueAsDefault=true
   desc="Coordinate space for the cone. Feed the UVs of the shape you're decorating, with any transforms baked in (scale/offset/rotation/skew), so the cone stays aligned to it. \r\nCentered by UVs Offset and normalized internally to get the per-pixel direction. Usually [0,1]."
   description="Coordinate space for the cone. Feed the UVs of the shape you're decorating, with any transforms baked in (scale/offset/rotation/skew), so the cone stays aligned to it. Centered by UVs Offset and normalized internally to get the per-pixel direction. Usually [0,1]."
   inputName=UVs
   inputType=FunctionInput_Vector2
  }
  n9 = FunctionInput() {
   bUsePreviewValueAsDefault=true
   desc="Subtracted from the UVs to center them on the shape. \r\nDefault (0.5, 0.5). Set (0, 0) if already centered, or whatever value re-centers your particular UVs."
   description="Subtracted from the UVs to center them on the shape. Default (0.5, 0.5). Set (0, 0) if already centered, or whatever value re-centers your particular UVs."
   inputName=UVs Offset
   inputType=FunctionInput_Vector2
   sortPriority=1
   previewValue=<0.5 0.5 0 0>
  }
  NamedRerouteDeclaration(None=Subtract(A=@n8, B=@n9)) {name=Centered UVs, nodeColor=<rgba 0 0 0 1>}
 ]
}
