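# UI material function that produces a scalar three-stop linear gradient with Start, Mid and End positions, presumably for blending three colors.
# It calls MF_UI_LinearGradient for Start-to-Mid and Mid-to-End, sharing rotation, center, offset and a Texture input that defaults to UV X.
# Each segment is remapped to half the range, then both are summed and saturated, placing Mid at 0.5.
# It outputs only the ramp, with no color inputs.
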MaterialFunction /FortUILibrary/MaterialFunctions/Gradients/MF_UI_LinearGradient_3Colors
graph {nodes [
  n1 = FunctionInput(Preview=ComponentMask(None=TextureCoordinate()) {r=true}) {bUsePreviewValueAsDefault=true, inputName=Texture, inputType=FunctionInput_Scalar, sortPriority=1, previewValue=<1 1 1 1>}
  n2 = FunctionInput() {bUsePreviewValueAsDefault=true, inputName=Gradient Mid, inputType=FunctionInput_Scalar, sortPriority=3, previewValue=<0.5 0.5 0.5 1>}
  n3 = FunctionInput() {bUsePreviewValueAsDefault=true, inputName=Rotation Angle, inputType=FunctionInput_Scalar, sortPriority=5}
  n4 = FunctionInput() {bUsePreviewValueAsDefault=true, inputName=Gradient Offset, inputType=FunctionInput_Scalar, sortPriority=7}
  n5 = FunctionInput() {bUsePreviewValueAsDefault=true, inputName=Rotation Center, inputType=FunctionInput_Vector2, sortPriority=6, previewValue=<0.5 0.5 0.5 1>}
  n6 = LinearInterpolate(Alpha=MaterialFunctionCall(
    Texture=@n1
    GradientStart=FunctionInput() {bUsePreviewValueAsDefault=true, inputName=Gradient Start, inputType=FunctionInput_Scalar, sortPriority=2}
    GradientEnd=@n2
    RotationAngle=@n3
    RotationCenter=@n5
    GradientOffset=@n4
   ) {materialFunction=&MF_UI_LinearGradient}.Result) {constB=0.5}
  n7 = FunctionInput() {bUsePreviewValueAsDefault=true, inputName=Gradient End, inputType=FunctionInput_Scalar, sortPriority=4, previewValue=<1 1 1 1>}
  n8 = Subtract(A=LinearInterpolate(Alpha=MaterialFunctionCall(Texture=@n1, GradientStart=@n2, GradientEnd=@n7, RotationAngle=@n3, RotationCenter=@n5, GradientOffset=@n4) {materialFunction=&MF_UI_LinearGradient}.Result) {constA=0.5}) {constB=0.5}
  FunctionOutput(None=Saturate(None=Add(A=@n6, B=@n8)))
 ]}
