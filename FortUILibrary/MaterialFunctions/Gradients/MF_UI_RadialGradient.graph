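# UI material function in the shader library that produces a radial gradient from UV, GradientStart and GradientEnd inputs.
# It calls MF_UI_Circle at zero size, so the circle's distance field and fill act as radial falloff sources.
# The Linear output inverts the saturated SDF and remaps it between start and end.
# The Smooth output passes the circle's Fill through MF_UI_LinearGradient with the same start and end.
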MaterialFunction /FortUILibrary/MaterialFunctions/Gradients/MF_UI_RadialGradient
graph {
 properties {bExposeToLibrary=true, description=Radial Gradient Function, previewMaterialDomain=MD_UI, libraryCategoriesText [UI_ShaderLibrary]}
 nodes [
  n1 = FunctionInput(Preview=TextureCoordinate()) {
   bUsePreviewValueAsDefault=true
   description=Plug in a TexCoord + a 2Vec to change the UV coordinates
   inputName=UV
   inputType=FunctionInput_Vector2
   sortPriority=1
  }
  n2 = Constant()
  n3 = MaterialFunctionCall(UVs=@n1, Size=@n2, GlowMin=@n2, GlowMax=Constant() {r=1}) {materialFunction=&MF_UI_Circle}
  n4 = OneMinus(None=Saturate(None=@n3.SDF))
  n5 = FunctionInput() {
   bUsePreviewValueAsDefault=true
   description="0~1 End point of Gradient set"
   inputName=GradientEnd
   inputType=FunctionInput_Scalar
   sortPriority=4
   previewValue=<1 0 0 1>
  }
  n6 = FunctionInput() {
   bUsePreviewValueAsDefault=true
   description="0~1 Start point of Gradient"
   inputName=GradientStart
   inputType=FunctionInput_Scalar
   sortPriority=3
  }
  FunctionOutput(None=Saturate(None=MaterialFunctionCall!(Input=@n4, Input Low=@n6, Input High=@n5) {materialFunction=/Engine/Functions/Engine_MaterialFunctions03/Math/RemapValueRangeNormalized}.Result)) {outputName=Linear, sortPriority=5}
  FunctionOutput(None=MaterialFunctionCall(Texture=@n3.Fill, GradientStart=@n6, GradientEnd=@n5) {materialFunction=&MF_UI_LinearGradient}.Result) {outputName=Smooth}
 ]
}
